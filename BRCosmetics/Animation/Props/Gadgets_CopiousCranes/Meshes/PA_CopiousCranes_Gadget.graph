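# Physics asset for the Copious Cranes gadget prop, giving its skeletal mesh capsule collision bodies.
# Six lettered crane bones (A–F) share one capsule shape, their six sheet bones use near-identical zero-length capsules, and root, offset and main each have their own.
# Constraints chain root to offset to main, then fan out from main to every crane and sheet as independent siblings, all with uniform swing and twist limit angles.
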PhysicsAsset /BRCosmetics/Animation/Props/Gadgets_CopiousCranes/Meshes/PA_CopiousCranes_Gadget
bodies {
 JNT_A_Crane [{shapeType=Capsule, length=6.7194, radius=9.7085, center=<-0.0001 0.4885 5.3032>, rotation=<rot 0 180 -78.9204>}]
 JNT_A_Sheet [{shapeType=Capsule, length=0, radius=8.2016, center=<-0.0455 -0.0036 0.2206>, rotation=<rot 0 -133.0561 -89.4269>}]
 JNT_B_Crane [{shapeType=Capsule, length=6.7194, radius=9.7085, center=<-0.0001 0.4885 5.3032>, rotation=<rot 0 180 -78.9204>}]
 JNT_B_Sheet [{shapeType=Capsule, length=0, radius=8.2016, center=<-0.0246 -0.0036 0.2206>, rotation=<rot 0 -133.0561 -89.4269>}]
 JNT_C_Crane [{shapeType=Capsule, length=6.7194, radius=9.7085, center=<-0.0001 0.4885 5.3032>, rotation=<rot 0 180 -78.9204>}]
 JNT_C_Sheet [{shapeType=Capsule, length=0, radius=8.2016, center=<-0.0455 -0.0036 0.2206>, rotation=<rot 0 -133.0561 -89.4269>}]
 JNT_D_Crane [{shapeType=Capsule, length=6.7194, radius=9.7085, center=<-0.0001 0.4885 5.3032>, rotation=<rot 0 180 -78.9204>}]
 JNT_D_Sheet [{shapeType=Capsule, length=0, radius=8.2016, center=<-0.0246 -0.0036 0.2206>, rotation=<rot 0 -133.0561 -89.4269>}]
 JNT_E_Crane [{shapeType=Capsule, length=6.7194, radius=9.7085, center=<-0.0001 0.4885 5.3032>, rotation=<rot 0 180 -78.9204>}]
 JNT_E_Sheet [{shapeType=Capsule, length=0, radius=8.2016, center=<-0.0455 -0.0036 0.2206>, rotation=<rot 0 -133.0561 -89.4269>}]
 JNT_F_Crane [{shapeType=Capsule, length=6.7194, radius=9.7085, center=<-0.0001 0.4885 5.3032>, rotation=<rot 0 180 -78.9204>}]
 JNT_F_Sheet [{shapeType=Capsule, length=0, radius=8.2016, center=<-0.0246 -0.0036 0.2206>, rotation=<rot 0 -133.0561 -89.4269>}]
 main [{shapeType=Capsule, length=8.6446, radius=7.3854, center=<0.0177 -0.1722 4.5328>, rotation=<rot 4.5102 93.9801 -90>}]
 offset [{shapeType=Capsule, length=0.505, radius=0.505, rotation=<rot 0 -90 0>}]
 root [{shapeType=Capsule, length=0.505, radius=19.7118, rotation=<rot 0 -90 0>}]
}
constraints [
 {bone1Name=offset, bone2Name=root, swing1LimitDegrees=45, swing2LimitDegrees=45, twistLimitDegrees=45}
 {bone1Name=main, bone2Name=offset, swing1LimitDegrees=45, swing2LimitDegrees=45, twistLimitDegrees=45}
 {bone1Name=JNT_A_Crane, bone2Name=main, swing1LimitDegrees=45, swing2LimitDegrees=45, twistLimitDegrees=45}
 {bone1Name=JNT_B_Crane, bone2Name=main, swing1LimitDegrees=45, swing2LimitDegrees=45, twistLimitDegrees=45}
 {bone1Name=JNT_C_Crane, bone2Name=main, swing1LimitDegrees=45, swing2LimitDegrees=45, twistLimitDegrees=45}
 {bone1Name=JNT_D_Crane, bone2Name=main, swing1LimitDegrees=45, swing2LimitDegrees=45, twistLimitDegrees=45}
 {bone1Name=JNT_E_Crane, bone2Name=main, swing1LimitDegrees=45, swing2LimitDegrees=45, twistLimitDegrees=45}
 {bone1Name=JNT_F_Crane, bone2Name=main, swing1LimitDegrees=45, swing2LimitDegrees=45, twistLimitDegrees=45}
 {bone1Name=JNT_A_Sheet, bone2Name=main, swing1LimitDegrees=45, swing2LimitDegrees=45, twistLimitDegrees=45}
 {bone1Name=JNT_B_Sheet, bone2Name=main, swing1LimitDegrees=45, swing2LimitDegrees=45, twistLimitDegrees=45}
 {bone1Name=JNT_C_Sheet, bone2Name=main, swing1LimitDegrees=45, swing2LimitDegrees=45, twistLimitDegrees=45}
 {bone1Name=JNT_D_Sheet, bone2Name=main, swing1LimitDegrees=45, swing2LimitDegrees=45, twistLimitDegrees=45}
 {bone1Name=JNT_E_Sheet, bone2Name=main, swing1LimitDegrees=45, swing2LimitDegrees=45, twistLimitDegrees=45}
 {bone1Name=JNT_F_Sheet, bone2Name=main, swing1LimitDegrees=45, swing2LimitDegrees=45, twistLimitDegrees=45}
]
